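# Niagara system for the Target Practice backpack cosmetic that provides its blue, pink and red lights at fixed offsets on the pack.
# Each color is an emitter that bursts one particle drawn by sprite and light renderers.
# Emitters copy User timing parameters, apparently Blueprint-fed for default and reactive blinking, and set their own blink-order vertex color value.
# Every update, a local scratch dynamic input multiplies each particle's base light radius into NewLightRadius.
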NiagaraSystem /BRCosmetics/Accessories/FORT_Backpacks/Backpack_M_MED_Target_Practice/Effects/NS_TargetPractice_BackpackLight
dependencies {usedDataInterfaces [], usedDynamicInputs [], usedModules [&SystemState]}
emitters {
 BlueLight {
  emitterUpdateScript [
   &EmitterState(Life Cycle Mode=::NewEnumerator1, Loop Duration=1, UseLoopDelay=false)
   &SpawnBurst_Instantaneous(Spawn Count=1, Use Spawn Probability=false)
   /SetVariables(
    Emitter.BlinkOrderVertColor=0.753
    Emitter.TP_GameTimeFromBP:NiagaraFloat=$User.TP_GameTimeFromBP
    Emitter.TP_LightsDefaultSpeed:NiagaraFloat=$User.TP_LightsDefaultSpeed
    Emitter.TP_LightsReactiveSpeed:NiagaraFloat=$User.TP_LightsReactiveSpeed
    Emitter.TP_ReactiveBlinkDuration:NiagaraFloat=$User.TP_ReactiveBlinkDuration
    Emitter.TP_ReactiveGlowDuration:NiagaraFloat=$User.TP_ReactiveGlowDuration
    Emitter.TP_TimeOffset:NiagaraFloat=$User.TP_TimeOffset
    Emitter.TP_UseVertexColorChannelForTimeOffset:NiagaraFloat=$User.TP_UseVertexColorChannelForTimeOFfset
    Emitter.TP_ValueToHideonReactive:NiagaraFloat=$User.TP_ValueToHideonReactive
   ) {isSetParameters=true, module=SetVariables_#1}
  ]
  particleSpawnScript [
   /Niagara/Modules/Spawn/Initialization/InitializeParticle(
    Color=<rgba 0.1202 3.0494 3.5 0>
    Module.Write Color=true
    Module.Write Lifetime=true
    Module.Write Mass=true
    Module.Write Material Random=true
    Module.Write Position=false
    Module.Write Scale=false
    Module.Write SpriteRotation=false
    Module.Write SpriteSize=true
    Sprite Size=<5 5>
   )
   /SetVariables(Particles.LightRadius=25) {isSetParameters=true, module=SetVariables_#2}
   &SystemLocation(MaskSpawnGroup=false, Offset=<0 -14.0119 -0.28>)
  ]
  particleUpdateScript [
   &ParticleState()
   &SolveForcesAndVelocity(Clamp Velocity=false, Limit Acceleration=false)
   /SetVariables(Particles.NewLightRadius:NiagaraFloat=&Multiply_Float(
     A=$Particles.LightRadius
     B=&self:MultiplyLightRadiusScratch()
    )) {isSetParameters=true, module=SetVariables_#3}
  ]
  rendererClasses [&NiagaraSpriteRendererProperties, &NiagaraLightRendererProperties]
  renderers [
   {rendererClass=&NiagaraSpriteRendererProperties}
   {rendererClass=&NiagaraLightRendererProperties, rendererIndex=1}
  ]
 }
 PinkLight {
  emitterUpdateScript [
   &EmitterState(Life Cycle Mode=::NewEnumerator1, Loop Duration=1, UseLoopDelay=false)
   &SpawnBurst_Instantaneous(Spawn Count=1, Use Spawn Probability=false)
   /SetVariables(
    Emitter.BlinkOrderVertColor=0.502
    Emitter.TP_GameTimeFromBP:NiagaraFloat=$User.TP_GameTimeFromBP
    Emitter.TP_LightsDefaultSpeed:NiagaraFloat=$User.TP_LightsDefaultSpeed
    Emitter.TP_LightsReactiveSpeed:NiagaraFloat=$User.TP_LightsReactiveSpeed
    Emitter.TP_ReactiveBlinkDuration:NiagaraFloat=$User.TP_ReactiveBlinkDuration
    Emitter.TP_ReactiveGlowDuration:NiagaraFloat=$User.TP_ReactiveGlowDuration
    Emitter.TP_TimeOffset:NiagaraFloat=$User.TP_TimeOffset
    Emitter.TP_UseVertexColorChannelForTimeOffset:NiagaraFloat=$User.TP_UseVertexColorChannelForTimeOFfset
    Emitter.TP_ValueToHideonReactive:NiagaraFloat=$User.TP_ValueToHideonReactive
    Emitter.TimingOffsetVertColor=0.502
   ) {isSetParameters=true, module=SetVariables_#1}
  ]
  particleSpawnScript [
   &InitializeParticle(
    Color=<rgba 3.5 0.1202 1.4452 0>
    Module.Write Color=true
    Module.Write Lifetime=true
    Module.Write Mass=true
    Module.Write Material Random=true
    Module.Write Position=false
    Module.Write Scale=false
    Module.Write SpriteRotation=false
    Module.Write SpriteSize=true
    Sprite Size=<5 5>
   )
   /SetVariables(Particles.LightRadius=30) {isSetParameters=true, module=SetVariables_#2}
   &SystemLocation(MaskSpawnGroup=false, Offset=<15.712 -0.867 17.8549>)
  ]
  particleUpdateScript [
   &ParticleState()
   &SolveForcesAndVelocity(Clamp Velocity=false, Limit Acceleration=false)
   /SetVariables(Particles.NewLightRadius:NiagaraFloat=&Multiply_Float(
     A=$Particles.LightRadius
     B=&self:MultiplyLightRadiusScratch()
    )) {isSetParameters=true, module=SetVariables_#3}
  ]
  rendererClasses [&NiagaraSpriteRendererProperties, &NiagaraLightRendererProperties]
  renderers [
   {rendererClass=&NiagaraSpriteRendererProperties}
   {rendererClass=&NiagaraLightRendererProperties, rendererIndex=1}
  ]
 }
 RedLight {
  emitterUpdateScript [
   &EmitterState(Life Cycle Mode=::NewEnumerator1, Loop Duration=1, UseLoopDelay=false)
   &SpawnBurst_Instantaneous(Spawn Count=1, Use Spawn Probability=false)
   /SetVariables(
    Emitter.BlinkOrderVertColor=0.251
    Emitter.TP_GameTimeFromBP:NiagaraFloat=$User.TP_GameTimeFromBP
    Emitter.TP_LightsDefaultSpeed:NiagaraFloat=$User.TP_LightsDefaultSpeed
    Emitter.TP_LightsReactiveSpeed:NiagaraFloat=$User.TP_LightsReactiveSpeed
    Emitter.TP_ReactiveBlinkDuration:NiagaraFloat=$User.TP_ReactiveBlinkDuration
    Emitter.TP_ReactiveGlowDuration:NiagaraFloat=$User.TP_ReactiveGlowDuration
    Emitter.TP_TimeOffset:NiagaraFloat=$User.TP_TimeOffset
    Emitter.TP_UseVertexColorChannelForTimeOffset:NiagaraFloat=$User.TP_UseVertexColorChannelForTimeOFfset
    Emitter.TP_ValueToHideonReactive:NiagaraFloat=$User.TP_ValueToHideonReactive
   ) {isSetParameters=true, module=SetVariables_#1}
  ]
  particleSpawnScript [
   &InitializeParticle(
    Color=<rgba 3.5 0.1202 0.1765 0>
    Module.Write Color=true
    Module.Write Lifetime=true
    Module.Write Mass=true
    Module.Write Material Random=true
    Module.Write Position=false
    Module.Write Scale=false
    Module.Write SpriteRotation=false
    Module.Write SpriteSize=true
    Sprite Size=<5 5>
   )
   /SetVariables(Particles.LightRadius=12) {isSetParameters=true, module=SetVariables_#2}
   &SystemLocation(MaskSpawnGroup=false, Offset=<9.5731 -1.4065 1.5781>)
  ]
  particleUpdateScript [
   &ParticleState()
   &SolveForcesAndVelocity(Clamp Velocity=false, Limit Acceleration=false)
   /SetVariables(Particles.NewLightRadius:NiagaraFloat=&Multiply_Float(
     A=$Particles.LightRadius
     B=&self:MultiplyLightRadiusScratch()
    )) {isSetParameters=true, module=SetVariables_#3}
  ]
  rendererClasses [&NiagaraSpriteRendererProperties, &NiagaraLightRendererProperties]
  renderers [
   {rendererClass=&NiagaraSpriteRendererProperties}
   {rendererClass=&NiagaraLightRendererProperties, rendererIndex=1}
  ]
 }
}
summary {
 systemName=NS_TargetPractice_BackpackLight
 emitters [
  {emitterName=PinkLight, rendererClasses [&NiagaraSpriteRendererProperties, &NiagaraLightRendererProperties]}
  {emitterName=RedLight, rendererClasses [&NiagaraSpriteRendererProperties, &NiagaraLightRendererProperties]}
  {emitterName=BlueLight, rendererClasses [&NiagaraSpriteRendererProperties, &NiagaraLightRendererProperties]}
 ]
 userVariables [
  User.TP_GameTimeFromBP {type=NiagaraFloat, defaultValue=0, description=""}
  User.TP_LightsDefaultSpeed {type=NiagaraFloat, defaultValue=0, description=""}
  User.TP_LightsReactiveSpeed {type=NiagaraFloat, defaultValue=0, description=""}
  User.TP_ReactiveBlinkDuration {type=NiagaraFloat, defaultValue=0, description=""}
  User.TP_ReactiveGlowDuration {type=NiagaraFloat, defaultValue=0, description=""}
  User.TP_TimeOffset {type=NiagaraFloat, defaultValue=0, description=""}
  User.TP_UseVertexColorChannelForTimeOFfset {type=NiagaraFloat, defaultValue=0, description=""}
  User.TP_ValueToHideonReactive {type=NiagaraFloat, defaultValue=0, description=""}
 ]
}
