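# Translucent, two-sided mesh-particle material for effects on the Target Practice backpack cosmetic.
# The backpack diffuse texture supplies base color, while particle color alpha drives opacity and opacity mask so particles can fade.
# These attributes pass through MF_TargetPractice_Reactive, then the accessories tech-art overrides function.
# Lighting uses volumetric per-vertex non-directional translucency, and a UV1 coordinate node is left unconnected.
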Material /BRCosmetics/Accessories/FORT_Backpacks/Backpack_M_MED_Target_Practice/Materials/M_TargetPractice_FX
graph {
 properties {
  bUseMaterialAttributes=true
  bUsedWithMeshParticles=true
  blendMode=BLEND_Translucent
  pixelDepthOffsetMode=PDOM_Legacy
  translucencyLightingMode=TLM_VolumetricPerVertexNonDirectional
  twoSided=true
  usedShadingModels=Default Lit
 }
 nodes [
  n1 = TextureSample() {texture=/BRCosmetics/Accessories/FORT_Backpacks/Backpack_M_MED_Target_Practice/Textures/T_M_MED_TargetPractice_Backpack_D}
  n2 = ParticleColor()
  MaterialFunctionCall(In=MaterialFunctionCall!(Material=MakeMaterialAttributes!(
     BaseColor=@n1.RGB
     Metallic
     Specular
     Roughness
     Anisotropy
     EmissiveColor
     Opacity=@n2.A
     OpacityMask=@n2.A
     Normal
     Tangent
     WorldPositionOffset
     SubsurfaceColor
     ClearCoat
     ClearCoatRoughness
     AmbientOcclusion
     Refraction
     CustomizedUVs_0
     CustomizedUVs_1
     CustomizedUVs_2
     CustomizedUVs_3
     CustomizedUVs_4
     CustomizedUVs_5
     CustomizedUVs_6
     CustomizedUVs_7
     PixelDepthOffset
     ShadingModel
     Displacement
    )) {materialFunction=/Game/Accessories/FORT_Backpacks/Backpack_M_MED_Target_Practice/Materials/MF_TargetPractice_Reactive}.Result) {materialFunction=&MF_TechArtOverrides_accessories}
  TextureCoordinate() {coordinateIndex=1}
 ]
}
