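# Physics asset for the Target Practice backpack cosmetic, giving its dangling arrow, paper and duck bones secondary motion.
# Capsule bodies form constrained chains rooted at the pack: a master arrow chain that branches to top and bottom arrows, left and right paper chains, and two duck chains.
# Swing and twist limits are tight along most arrow links and looser on the papers, while both duck root joints rotate freely with tight limits further down.
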PhysicsAsset /BRCosmetics/Accessories/FORT_Backpacks/Backpack_M_MED_Target_Practice/Meshes/M_MED_Target_Practice_Pack_Physics
bodies {
 dyn_arrow_bot [{shapeType=Capsule, length=8.223, radius=7.0785, center=<-2.6534 1.9947 0.2169>, rotation=<rot 0 -98.8934 -86.0891>}]
 dyn_arrow_top [{shapeType=Capsule, length=4, radius=3.75, center=<-2.9771 2.1652 0>}]
 dyn_duckA [{shapeType=Capsule, length=0.505, radius=0.505, center=<-0.004 -0.7695 10.1707>, rotation=<rot 0 -131.7111 -88.8565>}]
 dyn_duckA_01 [{shapeType=Capsule, length=1, radius=1, center=<0 0 0.5>}]
 dyn_duckA_02 [{shapeType=Capsule, length=2, radius=2, center=<0 0 2>, rotation=<rot 90 0 0>}]
 dyn_duckB [{shapeType=Capsule, length=0.505, radius=0.505, center=<-0.0159 0.5442 10.1169>, rotation=<rot 0 -68.4804 -89.3006>}]
 dyn_duckB_01 [{shapeType=Capsule, length=1, radius=1, center=<0 0 0.5>}]
 dyn_duckB_02 [{shapeType=Capsule, length=2.5, radius=2, center=<0 0 1.5>, rotation=<rot 90 0 0>}]
 dyn_master_arrow_01 [{shapeType=Capsule, length=5.5, radius=4.5, rotation=<rot 90 0 0>}]
 dyn_master_arrow_02 [{shapeType=Capsule, length=3.25, radius=3.25, rotation=<rot -34.55 0 0>}]
 dyn_master_arrow_03 [{shapeType=Capsule, length=3.75, radius=3, center=<1.1382 -5.4714 0>, rotation=<rot 0 0 -17.5165>}]
 dyn_master_arrow_04 [{shapeType=Capsule, length=1, radius=3.25, center=<-1.7 4.5 0>}]
 dyn_master_arrow_bot [{shapeType=Capsule, length=0.75, radius=6.4834, center=<0.2841 1.1709 -1.0309>, rotation=<rot 43.704 -90 10.9842>}]
 dyn_master_arrows [{shapeType=Capsule, length=1, radius=1, rotation=<rot 90 0 0>}]
 dyn_paper_01_l [{shapeType=Capsule, length=1, radius=1, center=<0 0 -0.5>}]
 dyn_paper_01_r [{shapeType=Capsule, length=1, radius=1, center=<0 0 -1>}]
 dyn_paper_l [{shapeType=Capsule, length=0.75, radius=1.75, center=<0 0 -3.4232>}]
 dyn_paper_r [{shapeType=Capsule, length=1.25, radius=3.25, center=<0 0 -3.1062>}]
 dyn_papercounter_l [{shapeType=Capsule, length=1.25, radius=2.25, center=<0 0 -0.7>}]
 dyn_papercounter_r [{shapeType=Capsule, length=2.5, radius=2.5, center=<0 0 -1>}]
 pack [{shapeType=Capsule, length=4.1315, radius=4.0434, center=<-0.1576 -2.6515 -2.6544>, rotation=<rot 0 89.8811 -59.1571>}]
}
constraints [
 {bone1Name=dyn_master_arrows, bone2Name=pack, swing1LimitDegrees=10, swing2LimitDegrees=10, twistLimitDegrees=3}
 {bone1Name=dyn_master_arrow_01, bone2Name=dyn_master_arrows, swing1LimitDegrees=5, swing2LimitDegrees=5, twistLimitDegrees=3}
 {bone1Name=dyn_master_arrow_02, bone2Name=dyn_master_arrow_01, swing1LimitDegrees=10, swing2LimitDegrees=10, twistLimitDegrees=3}
 {bone1Name=dyn_master_arrow_03, bone2Name=dyn_master_arrow_02, swing1LimitDegrees=10, swing2LimitDegrees=10, twistLimitDegrees=3}
 {bone1Name=dyn_master_arrow_04, bone2Name=dyn_master_arrow_03, swing1LimitDegrees=10, swing2LimitDegrees=10, twistLimitDegrees=3}
 {bone1Name=dyn_arrow_top, bone2Name=dyn_master_arrow_04, swing1LimitDegrees=2, swing2LimitDegrees=20, twistLimitDegrees=1}
 {bone1Name=dyn_master_arrow_bot, bone2Name=dyn_master_arrow_02, swing1LimitDegrees=1, swing2LimitDegrees=1, twistLimitDegrees=1}
 {bone1Name=dyn_arrow_bot, bone2Name=dyn_master_arrow_bot, swing1LimitDegrees=5, swing2LimitDegrees=30, twistLimitDegrees=1}
 {bone1Name=dyn_papercounter_l, bone2Name=pack, swing1LimitDegrees=35, swing2LimitDegrees=45, twistLimitDegrees=20}
 {bone1Name=dyn_papercounter_r, bone2Name=pack, swing1LimitDegrees=35, swing2LimitDegrees=45, twistLimitDegrees=20}
 {bone1Name=dyn_paper_l, bone2Name=dyn_papercounter_l, swing1LimitDegrees=35, swing2LimitDegrees=40, twistLimitDegrees=20}
 {bone1Name=dyn_paper_r, bone2Name=dyn_papercounter_r, swing1LimitDegrees=35, swing2LimitDegrees=30, twistLimitDegrees=20}
 {bone1Name=dyn_paper_01_r, bone2Name=dyn_paper_r, swing1LimitDegrees=45, swing2LimitDegrees=45, twistLimitDegrees=1}
 {bone1Name=dyn_paper_01_l, bone2Name=dyn_paper_l, swing1LimitDegrees=45, swing2LimitDegrees=65, twistLimitDegrees=1}
 {
  bone1Name=dyn_duckA
  bone2Name=pack
  swing1LimitDegrees=45
  swing1Motion=Free
  swing2LimitDegrees=45
  swing2Motion=Free
  twistLimitDegrees=45
  twistMotion=Free
 }
 {
  bone1Name=dyn_duckB
  bone2Name=pack
  swing1LimitDegrees=45
  swing1Motion=Free
  swing2LimitDegrees=45
  swing2Motion=Free
  twistLimitDegrees=45
  twistMotion=Free
 }
 {bone1Name=dyn_duckA_01, bone2Name=dyn_duckA, swing1LimitDegrees=5, swing2LimitDegrees=10, twistLimitDegrees=1}
 {bone1Name=dyn_duckB_01, bone2Name=dyn_duckB, swing1LimitDegrees=5, swing2LimitDegrees=5, twistLimitDegrees=1}
 {bone1Name=dyn_duckA_02, bone2Name=dyn_duckA_01, swing1LimitDegrees=2, swing2LimitDegrees=20, twistLimitDegrees=1}
 {bone1Name=dyn_duckB_02, bone2Name=dyn_duckB_01, swing1LimitDegrees=2, swing2LimitDegrees=20, twistLimitDegrees=1}
]
