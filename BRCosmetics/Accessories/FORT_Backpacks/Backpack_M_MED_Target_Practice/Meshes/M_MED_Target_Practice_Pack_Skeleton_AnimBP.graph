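# Anim Blueprint for the Target Practice back bling, which moves two duck targets and tilts the pack with the wearer's spine bend.
# Each duck's pitch steps up on alternating clock seconds, then rewinds at its limit.
# A watched kill-stat increase triggers a fading sine wobble on the ducks' child bones, followed by staggered rewinds.
# EventGraph-computed rotators drive chained additive bone modifiers, finished by a RigidBody simulation.
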AnimBlueprint /BRCosmetics/Accessories/FORT_Backpacks/Backpack_M_MED_Target_Practice/Meshes/M_MED_Target_Practice_Pack_Skeleton_AnimBP
target_skeleton=&/M_MED_Target_Practice_Pack_Skeleton
graphs [
 AnimGraph {nodes [
   n1 = "Transform (Modify) Bone - Bone: pack"(ComponentPose=Local To Component(LocalPose=Local Space Ref Pose()), Translation="0,0,0", Rotation=Get backpack_rotate()) {rotationMode=BMM_Additive, rotationSpace=BCS_BoneSpace, alphaBoolBlend {blendOption=Linear}, boneToModify {boneName=pack}}
   n2 = "Transform (Modify) Bone - Bone: dyn_duckA"(ComponentPose=@n1, Alpha=1, Translation="0,0,0", Rotation=Get DuckA() -> (DuckA:Rotator), Scale="1,1,1") {rotationMode=BMM_Additive, rotationSpace=BCS_BoneSpace, alphaBoolBlend {blendOption=Linear}, boneToModify {boneName=dyn_duckA}}
   n3 = "Transform (Modify) Bone - Bone: dyn_duckB"(ComponentPose=@n2, Alpha=1, Translation="0,0,0", Rotation=Get DuckB() -> (DuckB:Rotator), Scale="1,1,1") {rotationMode=BMM_Additive, rotationSpace=BCS_BoneSpace, alphaBoolBlend {blendOption=Linear}, boneToModify {boneName=dyn_duckB}}
   n4 = "Transform (Modify) Bone - Bone: dyn_duckA_01"(
    ComponentPose=@n3
    Alpha=Get Duck_Alpha() -> (Duck_Alpha:double)
    Translation="0,0,0"
    Rotation=Get DuckA_Rotator() -> (DuckA_Rotator:Rotator)
    Scale="1,1,1"
   ) {rotationMode=BMM_Additive, rotationSpace=BCS_BoneSpace, alphaBoolBlend {blendOption=Linear}, boneToModify {boneName=dyn_duckA_01}}
   n5 = "Transform (Modify) Bone - Bone: dyn_duckB_01"(ComponentPose=@n4, Alpha=Get Duck_Alpha(), Translation="0,0,0", Rotation=Get DuckB_Rotator() -> (DuckB_Rotator:Rotator), Scale="1,1,1") {rotationMode=BMM_Additive, rotationSpace=BCS_BoneSpace, alphaBoolBlend {blendOption=Linear}, boneToModify {boneName=dyn_duckB_01}}
   n6 = RigidBody(ComponentPose=@n5) {
    bClampLinearTranslationLimitToRefPose=true
    bOverrideWorldGravity=true
    overridePhysicsAsset=&/M_MED_Target_Practice_Pack_Physics
    simulationSpace=BaseBoneSpace
    alphaBoolBlend {blendOption=Linear}
    baseBoneRef {boneName=root}
    componentAppliedLinearAccClamp=<1100 1100 1100>
    componentLinearAccScale=<0.5 0.5 0.5>
    componentLinearVelScale=<0.5 0.5 0.5>
    overrideWorldGravity {z=-980}
   }
   Output Pose(Result=Component To Local(ComponentPose=@n6)) {layerGroup=None}
  ]}
 EventGraph {nodes [
   n1 = Event BlueprintUpdateAnimation()
   > Get!(execute:Exec) -> (then:Exec, else:Exec, Skel_Mesh:Skeletal Mesh Component Object Reference)
   > Get!(execute) -> (then, else, Pawn:Fort Player Pawn Object Reference)
   > n2 = Sequence!(execute) -> (then_0, then_1, then_2:Exec, then_3:Exec, then_4:Exec, then_5:Exec, then_6:Exec, then_7:Exec, then_8:Exec, then_9:Exec)
   Branch(execute=@n2.then_3, Condition=Get DuckA_Rewind() -> (DuckA_Rewind:bool))
   > n3 = Set DuckA_Val(DuckA_Val:double=float - float(A=Get DuckA_Val() -> (DuckA_Val:double), B=3) {is=title K2Node_CallFunction}) -> (Output_Get:double)
   > Branch(Condition="float <= float"!(A:double=@n3, B:double=15) -> (ReturnValue:bool) {is=title K2Node_CallFunction})
   > Set DuckA_Rewind(DuckA_Rewind:bool=false) -> (Output_Get:bool)
   Branch(execute=@n2.then_8, Condition=Get DuckB_Rewind() -> (DuckB_Rewind:bool))
   > n4 = Set DuckB_Val(DuckB_Val:double=float - float(A=Get DuckB_Val() -> (DuckB_Val:double), B=3)) -> (Output_Get:double)
   > Branch(Condition="float <= float"(A=@n4, B=-3.5))
   > Set DuckB_Rewind(DuckB_Rewind:bool=false) -> (Output_Get:bool)
   n5 = Branch(execute=@n2.then_6, Condition=Get Active_Kill() -> (Active_Kill:bool))
   n6 = "Sin (Degrees)"!(A:double=float * float(A=GetGameTimeInSeconds(), B=750)) -> (ReturnValue:double)
   > Set DuckA_Rotator(DuckA_Rotator:Rotator=MakeRotator(Pitch=float * float(A=@n6, B=50))) -> (Output_Get:Rotator)
   n7 = "Sin (Degrees)"(A=float * float(A=GetGameTimeInSeconds(), B=775))
   > Set DuckB_Rotator(DuckB_Rotator:Rotator=MakeRotator(Pitch=float * float(A=@n7, B=55))) -> (Output_Get:Rotator)
   > Set Duck_Timer(Duck_Timer:double=float + float(A=Get Duck_Timer() -> (Duck_Timer:double), B=1) {is=title K2Node_CommutativeAssociativeBinaryOperator}) -> (Output_Get:double)
   n8 = Get Duck_Timer_Length() -> (Duck_Timer_Length:double)
   n9 = Get Duck_Timer()
   > Set Duck_Alpha(Duck_Alpha:double=MapRangeClamped(Value=float - float(A=@n8, B=@n9), InRangeA=Get Duck_Timer_Length(), InRangeB=0, OutRangeA=1, OutRangeB=0)) -> (Output_Get:double)
   > Branch(Condition="Equal (Float)"!(A:double=Get Duck_Alpha(), B:double=0) -> (ReturnValue:bool) {is=title K2Node_CallFunction})
   > Set Active_Kill(Active_Kill:bool=false) -> (Output_Get:bool)
   > Set Duck_Timer(Duck_Timer=0)
   > Delay(Duration=0.1)
   > Set Duck_Alpha(Duck_Alpha=1)
   > Set DuckA_Rewind(DuckA_Rewind=true)
   > Delay(Duration=0.4)
   > Set DuckB_Rewind(DuckB_Rewind=true)
   n10 = MakeRotator()
   Set DuckA_Rotator(execute=@n5.else, DuckA_Rotator=@n10)
   > Set DuckB_Rotator(DuckB_Rotator=@n10)
   Set DuckA(execute=@n2.then_1, DuckA:Rotator=MakeRotator(Pitch=Get DuckA_Val())) -> (Output_Get:Rotator)
   n11 = Get Skel_Mesh()
   n12 = BreakRotator(InRot=BreakTransform(InTransform=GetDeltaTransformFromRefPose(self=@n11, BoneName=spine_05, BaseName=spine_01)).Rotation)
   Set backpack_rotate(execute=@n2.then_5, backpack_rotate=MakeRotator(Roll=MapRangeClamped(Value=@n12.Yaw, InRangeA=-5, InRangeB=-30, OutRangeB=15)))
   n13 = Get Second!(A:Date Time Structure=Now!() -> (ReturnValue:Date Time Structure)) -> (ReturnValue:int=0) {is=title K2Node_CallFunction}
   n14 = Branch(
    execute=@n2.then_7
    Condition="Not Equal (Integer)"!(A:int=@n13, B:int=Get Timer() -> (Timer:int=0)) -> (ReturnValue:bool) {is=title K2Node_CallFunction}
   )
   > Set Timer(Timer:int=@n13) -> (Output_Get:int=0)
   > Set Eval_Ducks(Eval_Ducks:bool=true) -> (Output_Get:bool)
   Set Eval_Ducks(execute=@n14.else, Eval_Ducks=false)
   Set DuckB(execute=@n2.then_2, DuckB:Rotator=MakeRotator(Pitch=Get DuckB_Val())) -> (Output_Get:Rotator)
   Branch(execute=@n2.then_0, Condition=Get Eval_Ducks() -> (Eval_Ducks:bool))
   n15 = Get Second(A=Now()) -> (ReturnValue=0)
   > n16 = Branch(Condition="Equal (Integer)"(A="% (Integer)"(A=@n15, B=2) -> (ReturnValue=0), B=0) {is=title K2Node_CallFunction})
   > Branch(Condition=NOT Boolean(A=Get DuckA_Rewind()))
   n17 = Get DuckA_Val()
   > Set DuckA_Val(DuckA_Val=float + float(A=@n17, B=3.5))
   > Branch(Condition="float >= float"(A=@n17, B=100) {is=title K2Node_CallFunction})
   > Set DuckA_Rewind(DuckA_Rewind=true)
   Branch(execute=@n16.else, Condition=NOT Boolean(A=Get DuckB_Rewind()))
   n18 = Get DuckB_Val()
   > Set DuckB_Val(DuckB_Val=float + float(A=3.5, B=@n18))
   > Branch(Condition="float >= float"(A=@n18, B=100))
   > Set DuckB_Rewind(DuckB_Rewind=true)
   Event BlueprintInitializeAnimation()
   > Set Skel_Mesh(Skel_Mesh=GetOwnerSkeletalMeshComponent())
   > n19 = Cast To FortPlayerPawn(Object=TryGetFortPawn())
   > Set Pawn(Pawn:Fort Player Pawn Object Reference=@n19) -> (Output_Get:Fort Player Pawn Object Reference)
   > n20 = WatchCosmeticStat!(
    execute:Exec
    Pawn:Fort Player Pawn Object Reference=Get Pawn() -> (Pawn:Fort Player Pawn Object Reference)
    StatTag:Fort Stat Manager Tag Structure=(TagName="StatManager.AthenaKills")
    DefaultValue:int=0
    bBroadcastInitialValue:bool=false
    InitialValueBroadcastDelay:float=-1
    ChangeValueBroadcastDelay:float=-1
    bConsumeWaitingNotifications:bool=true
   ) -> (then:Exec, OnStatChanged:Exec, StatName:Name, StatValue:int) {is=title K2Node_AsyncAction}
   Set Did_Kill(execute=@n20.OnStatChanged, Did_Kill:bool=false) -> (Output_Get:bool)
   > n21 = Set KillCount(KillCount:double="To Float (Integer)"(InInt=@n20.StatValue)) -> (Output_Get:double)
   > Branch(Condition="float > float"(A=@n21, B=Get Last_KillCount() -> (Last_KillCount:double)) {is=title K2Node_CallFunction})
   > Set Active_Kill(Active_Kill=true)
   > Set Last_KillCount(Last_KillCount:double=@n21) -> (Output_Get:double)
   Branch(Condition="float > float"(A=Get KillCount() -> (KillCount:double), B=Get Last_KillCount()))
   > Set Active_Kill(Active_Kill=true)
   > Set Duck_Timer(Duck_Timer=0)
   TryGetPawnOwner()
   Reroute Node(InputPin:float=@n1.DeltaTimeX) -> (OutputPin:float)
  ]}
]
variables [
 Active_Kill {type=boolean}
 backpack_rotate {type=Rotator}
 Did_Kill {type=boolean}
 Duck_Alpha {type=number, default=1}
 Duck_Timer {type=number}
 Duck_Timer_Length {type=number, default=80}
 DuckA {type=Rotator}
 DuckA_Rewind {type=boolean}
 DuckA_Rotator {type=Rotator}
 DuckA_Val {type=number, default=30}
 DuckB {type=Rotator}
 DuckB_Rewind {type=boolean}
 DuckB_Rotator {type=Rotator}
 DuckB_Val {type=number}
 Eval_Ducks {type=boolean}
 KillCount {type=number}
 Last_KillCount {type=number}
 Pawn {type=FortPlayerPawn}
 Skel_Mesh {type=SkeletalMeshComponent}
 Timer {type=integer}
]
